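; Deployable MedCloud smoke building actor: a timed cloud with an AoE sphere that runs a server heal tick on overlapping pawns.
; Scalable-float rows set radius, lifespan and tick rate, while replicated SetFXMode switches Niagara effects and death cues.
; Smoke fades via timestamps in sphere custom primitive data and a post-process material.
; Optional client-side overlap tracking through a pawn component shows one sphere among overlapping clouds, and water entry fires throttled splashes.
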(event EventMulticast_HandleBinding
  (params (Actor :type Actor%20Object%20Reference) (HitComp :type Primitive%20Component%20Object%20Reference))
  ;[1 1 1] Handle Binding
  (|CallParentFunction Actor HitComp))

(event Custom|SetLocationFromServer
  (params (ServerLocation :type Vector))
  ;[1 1 1] Set Location From Server
  (Transformation|SetActorLocation self ServerLocation false true))

(event Building|EventOnBuildingActorInitialized
  (params (InitializationReason :type EFortBuildingInitializationReason%20Enum :default "StaticallyPlaced") (BuildingPersistentState :type EFortBuildingPersistentState%20Enum :default "Default"))
  (bind _row_lifespan (Variables|Default|GetRow_Lifespan))
  (bind _returnvalue (ScalableFloat|GetValueatLevel _row_lifespan))
  ;[1 1 1] On Building Initialized
  (|CallParentFunction InitializationReason BuildingPersistentState)
  (Components|Sphere|SetSphereRadius (Variables|Default|GetAoECollision) (ScalableFloat|GetValueatLevel (Variables|Default|GetRow_AoeRadius)) true)
  (AbilitySystem|SetTouchComponentforAbilityDeliveryInfo (Variables|Default|GetAoECollision))
  (Utilities|Time|SetTimerbyEvent (delegate Custom|KillSelf) _returnvalue)
  (Variables|Default|SetSetFXMode 1)
  (Variables|Default|SetEffectLength _returnvalue))

(event Custom|KillSelf
  ;[1 1 1] On Building Initialized
  (Variables|Default|SetDiedFromTimeOut true)
  (Gameplay|Die :self self :DamageCauser self)
  (Variables|Default|SetSetFXMode 0))

(event EventBeginPlay
  ;[1 1 1] Begin Play
  (|CallParentFunction)
  (Utilities|FlowControl|Sequence
    (:then_0
      (CallFunction|ResetEnterWaterEvent)
      (Audio|Components|Audio|Play (Variables|Default|GetEruptStartSound) 0.0)
      (CallFunction|SmokeFade true)
      (Networking|SwitchHasAuthority
        (:Authority
          (Health|BindEventtoOnBuildingHealthChanged (delegate Custom|HandleHealthChanged))
          (bind _returnvalue (Utilities|Time|SetTimerbyEvent (EventDispatchers|CreateEvent :!event Custom|ApplyHealGE) (ScalableFloat|GetValueatLevel (Variables|Default|GetRow%20ApplicationTickRate)) true))
          (Variables|Default|SetApplyHealGEPulseTH _returnvalue))))
    (:then_1
      (Variables|Default|SetOptimizeMultiOverlap (ScalableFloat|GetValueasBool (Variables|Default|GetSF_ToggleSphereVisibility)))
      (if (not (Variables|Default|GetOptimizeMultiOverlap))
        (Collision|SetCollisionResponsetoChannel (Variables|Default|GetAoECollision) "ECC_WorldDynamic" "ECR_Ignore")))
    (:then_2
      (Networking|SwitchHasAuthority
        (:Remote
          (if (not (Variables|Default|GetbResumeSimulation))
            (CallFunction|StopSim (Variables|Default|GetStopSimHitResult))))))
    (:then_3
      (Networking|CanExecuteCosmeticEvents
        (:True
          (bind _bsuccessfullyactivated (Class|AutofireShroudComponent|ActivateShroudDetectors (Variables|Default|GetAutofireShroudComponent) (Variables|/Script/Engine.SphereComponent|GetSphereRadius (Variables|Default|GetAoECollision))))
          (if (not _bsuccessfullyactivated)
            (Collision|SetCollisionResponsetoChannel (Variables|Default|GetAoECollision) "ECC_Visibility" "ECR_Block")
            (Fortnite|SetCanbeDamaged 0 true)))))))

(event Custom|HandleHealthChanged
  ;[1 1 1] Begin Play
  ;; Handle health changed
  (if (<= (Health|GetHealth) 0)
    (Variables|Default|SetSetFXMode 0)))

(event Custom|PlayFX
  ;[1 1 1] FX
  (bind _returnvalue (Niagara|SpawnSystemAttached (Niagara|GetNiagaraSystemAsset (Variables|Default|GetFX_Form)) (Variables|Default|GetStaticMesh)))
  (Niagara|SetNiagaraVariableByString(LinearColor) _returnvalue "SmokeColor" (Variables|SmokeFX|GetSmokeColor))
  (Niagara|SetNiagaraVariableByString(Float) _returnvalue "NightMultiplier" (Variables|SmokeFX|GetSmoke_Night_Multiplier))
  (Components|Activation|Activate (Variables|Default|GetFX) true)
  (Utilities|Time|SetTimerbyEvent (EventDispatchers|CreateEvent :!event Custom|ShowAudioIndicator) 0.75 true true))

(event Custom|ShowAudioIndicator
  ;[1 1 1] FX
  (Fortnite|ShowSoundIndicatorLocal :TrackedActor self :Location (Transformation|GetActorLocation self) :MaxAudibleDistance (Variables|Default|GetAV_Distance) :IndicatorType "Generic" :Tint "(R=0.000000,G=0.000000,B=0.000000,A=0.000000)" :OverrideIcon (Variables|Default|GetAV_Icon) :AttenuationMaxRangeMultiplier 1.0 :AttenuationIntensityMultiplier 1.0))

(event Custom|EndFX
  ;[1 1 1] FX
  (Audio|Components|Audio|FadeOut (Variables|Default|GetEruptLoopSound) 3.0 0.0 "Linear" :!class /Script/Engine.AudioComponent)
  (Components|Activation|Deactivate (Variables|Default|GetFX))
  (Audio|Components|Audio|SetSound (Variables|Default|GetEruptStartSound) "/MedCloud/Assets/Audio/MedCloud_MedCloud_EruptStop_Cue.MedCloud_MedCloud_EruptStop_Cue")
  (Default|DeactivateShroudDetectors (Variables|Default|GetAutofireShroudComponent)))

(event Custom|Update%20FX%20for%20Water
  (params (Water%20Body :type Fort%20Water%20Body%20Actor%20Object%20Reference) (IsInWater :type Boolean))
  (bind _fx (Variables|Default|GetFX))
  ;[1 1 1] Updates Spray FX to account for water
  (bind _outwatersurfaceinfo (WaterBody|GetWaterSurfaceInfoStruct Water%20Body (Transformation|GetActorLocation) true false false))
  (Effects|Components|ParticleSystem|SetBoolParameter _fx "bisInWater" IsInWater)
  (Effects|Components|ParticleSystem|SetFloatParameter _fx "WaterBodyIndex" (Math|Conversions|ToFloat(Integer) (.WaterBodyIdx (Utilities|Struct|BreakWaterSurfaceInfo _outwatersurfaceinfo))))
  (Effects|Components|ParticleSystem|SetFloatParameter _fx "WaterDepth" (.WaterDepth (Utilities|Struct|BreakWaterSurfaceInfo _outwatersurfaceinfo)))
  (Effects|Components|ParticleSystem|SetFloatParameter _fx "WaveAttenuationFactor" (Variables|/Script/Water.WaterBodyComponent|GetTargetWaveMaskDepth (Water|GetWaterBodyComponent Water%20Body)))
  (Effects|Components|ParticleSystem|SetVectorParameter _fx "WaterPlaneLocation" (Utilities|Struct|BreakWaterSurfaceInfo _outwatersurfaceinfo))
  (Effects|Components|ParticleSystem|SetVectorParameter _fx "WaterPlaneNormal" (.WaterPlaneNormal (Utilities|Struct|BreakWaterSurfaceInfo _outwatersurfaceinfo))))

(event Custom|PlayDeathFX
  (bind _self self)
  (bind _returnvalue (Ability|EffectContext|GetInstigatorActor))
  (bind _returnvalue_location (Math|Transform|BreakTransform (Transformation|GetActorTransform _self)))
  (bind _returnvalue_2 (Ability|GameplayCue|MakeGameplayCueParameters :NormalizedMagnitude 0.0 :RawMagnitude 0.0 :Location _returnvalue_location :Normal "0,0,1.000000" :Instigator _returnvalue :EffectCauser _self :SourceObject _self :GameplayEffectLevel 1 :AbilityLevel 1 :bReplicateLocationWhenUsingMinimalRepProxy false))
  ;[1 1 1] Play death FX
  (CallFunction|SmokeFade false)
  (Niagara|SetNiagaraVariable(Bool) (Variables|Default|GetFX) "FadeOut" true)
  (Rendering|SetVisibility (Variables|Default|GetStaticMesh) false false)
  (Networking|SwitchHasAuthority
    (:Authority
      (Ability|SendGameplayEventtoActor self (Variables|Default|GetEvent_ShieldGeneratorDied))
      (if (Variables|Default|GetDiedFromTimeOut)
        (Ability|ExecuteGameplayCueWithParamsReplicated (Variables|Default|GetTag_GC_TimedOut) _returnvalue_2)
        (else
          (Ability|ExecuteGameplayCueWithParamsReplicated (Variables|Default|GetTag_GC_Destroyed) _returnvalue_2))))))

(event EventMulticast_PlayHitFX
  (params (HitResult :type Hit%20Result%20Structure))
  ;[1 1 1] Bounce Land FX
  (Ability|ExecuteGameplayCueWithParamsReplicated (Variables|Default|GetTag_GC_HitWorld) (GameplayCue|MakeGameplayCueParametersfromHitResult HitResult)))

(event Custom|ApplyHealGE
  ; Manual Check to pass heal context for Quest team
  (for _array_element (Collision|GetOverlappingActors (Variables|Default|GetAoECollision) "/Script/Engine.Pawn" :!class /Script/Engine.PrimitiveComponent)
    (bind _asfort_player_pawn (Utilities|Casting|CastToFortPlayerPawn _array_element))
    (bind _asgameplay_tag_asset_interface (Utilities|Casting|CastToGameplayTagAssetInterface _asfort_player_pawn))
    (if (GameplayTags|HasMatchingGameplayTag _asgameplay_tag_asset_interface "(TagName=\"Gameplay.Status.DBNO\")")
      _)))

(event OnEnterWaterEvent(FortWaterInteraction)
  (params (WaterBody :type Fort%20Water%20Body%20Actor%20Object%20Reference) (WaterInteractionComponent :type Fort%20Water%20Interaction%20Component%20Object%20Reference) (bIsFirstBody :type Boolean))
  (bind _returnvalue_location (Math|Transform|BreakTransform (Transformation|GetActorTransform self)))
  ;[1 1 1] Water FX
  (CallFunction|UpdateFXForWater WaterBody true)
  (Utilities|FlowControl|DoOnce false
    (:Completed
      (Ability|ExecuteGameplayCueWithParamsReplicated (Variables|Default|GetTag_GC_HitWater) (Ability|GameplayCue|MakeGameplayCueParameters :NormalizedMagnitude 0.0 :RawMagnitude 0.0 :Location _returnvalue_location :Normal "0,0,1.000000" :Instigator (Ability|EffectContext|GetInstigatorActor) :EffectCauser self :SourceObject self :GameplayEffectLevel 1 :AbilityLevel 1 :bReplicateLocationWhenUsingMinimalRepProxy false))
      (Utilities|Time|SetTimerbyEvent (EventDispatchers|CreateEvent :!event Custom|ResetEnterWaterEvent) (ScalableFloat|GetValueatLevel (Variables|Default|GetRow_WaterSplashFXCooldown))))) ; Timed gate to prevent FX spam in super wavy waters
)

(event Custom|ResetEnterWaterEvent
  (bind _returnvalue_location (Math|Transform|BreakTransform (Transformation|GetActorTransform self)))
  ;[1 1 1] Water FX
  (Utilities|FlowControl|DoOnce false
    (:Completed
      (Ability|ExecuteGameplayCueWithParamsReplicated (Variables|Default|GetTag_GC_HitWater) (Ability|GameplayCue|MakeGameplayCueParameters :NormalizedMagnitude 0.0 :RawMagnitude 0.0 :Location _returnvalue_location :Normal "0,0,1.000000" :Instigator (Ability|EffectContext|GetInstigatorActor) :EffectCauser self :SourceObject self :GameplayEffectLevel 1 :AbilityLevel 1 :bReplicateLocationWhenUsingMinimalRepProxy false))
      (Utilities|Time|SetTimerbyEvent (EventDispatchers|CreateEvent :!event Custom|ResetEnterWaterEvent) (ScalableFloat|GetValueatLevel (Variables|Default|GetRow_WaterSplashFXCooldown))))) ; Timed gate to prevent FX spam in super wavy waters
)

(event OnExitWaterEvent(FortWaterInteraction)
  (params (WaterBody :type Fort%20Water%20Body%20Actor%20Object%20Reference) (WaterInteractionComponent :type Fort%20Water%20Interaction%20Component%20Object%20Reference) (bIsLastBody :type Boolean))
  ;[1 1 1] On Exit Water
  (CallFunction|UpdateFXForWater WaterBody))

(event OnComponentBeginOverlap(AoECollision)
  (params (OverlappedComponent :type Primitive%20Component%20Object%20Reference) (OtherActor :type Actor%20Object%20Reference) (OtherComp :type Primitive%20Component%20Object%20Reference) (OtherBodyIndex :type Integer) (bFromSweep :type Boolean) (SweepResult :type Hit%20Result%20Structure%20%28by%20ref%29))
  ; Handle pawn overlap optimization
  (if (and (Variables|Default|GetOptimizeMultiOverlap) (not (Networking|HasAuthority)))
    (bind _asplayer_pawn_athena (Utilities|Casting|CastToPlayerPawn_Athena OtherActor))
    (Utilities|IsValid (Actor|GetComponentbyClass _asplayer_pawn_athena "/MedCloud/Gameplay/Items/BP_SmokeCloud_PawnComp.BP_SmokeCloud_PawnComp_C")
      (:Is%20Valid
        (Utilities|IsValid (Variables|/MedCloud/Gameplay/Items/BP_SmokeCloud_PawnComp.BP_SmokeCloud_PawnComp_C|GetOwnerPawn (Actor|GetComponentbyClass _asplayer_pawn_athena "/MedCloud/Gameplay/Items/BP_SmokeCloud_PawnComp.BP_SmokeCloud_PawnComp_C"))
          (:Is%20Valid
            (Utilities|Array|AddUnique (Variables|/MedCloud/Gameplay/Items/BP_SmokeCloud_PawnComp.BP_SmokeCloud_PawnComp_C|GetOverlappingMedCloud (Actor|GetComponentbyClass _asplayer_pawn_athena "/MedCloud/Gameplay/Items/BP_SmokeCloud_PawnComp.BP_SmokeCloud_PawnComp_C")) self)
            (CallFunction|RefreshforMultiCloud (Actor|GetComponentbyClass _asplayer_pawn_athena "/MedCloud/Gameplay/Items/BP_SmokeCloud_PawnComp.BP_SmokeCloud_PawnComp_C"))))))))

(event OnComponentEndOverlap(AoECollision)
  (params (OverlappedComponent :type Primitive%20Component%20Object%20Reference) (OtherActor :type Actor%20Object%20Reference) (OtherComp :type Primitive%20Component%20Object%20Reference) (OtherBodyIndex :type Integer))
  ; Handle pawn overlap optimization
  (if (and (Variables|Default|GetOptimizeMultiOverlap) (not (Networking|HasAuthority)))
    (bind _asplayer_pawn_athena (Utilities|Casting|CastToPlayerPawn_Athena OtherActor))
    (Utilities|IsValid (Actor|GetComponentbyClass _asplayer_pawn_athena "/MedCloud/Gameplay/Items/BP_SmokeCloud_PawnComp.BP_SmokeCloud_PawnComp_C")
      (:Is%20Valid
        (Utilities|IsValid (Variables|/MedCloud/Gameplay/Items/BP_SmokeCloud_PawnComp.BP_SmokeCloud_PawnComp_C|GetOwnerPawn (Actor|GetComponentbyClass _asplayer_pawn_athena "/MedCloud/Gameplay/Items/BP_SmokeCloud_PawnComp.BP_SmokeCloud_PawnComp_C"))
          (:Is%20Valid
            (Utilities|Array|RemoveItem (Variables|/MedCloud/Gameplay/Items/BP_SmokeCloud_PawnComp.BP_SmokeCloud_PawnComp_C|GetOverlappingMedCloud (Actor|GetComponentbyClass _asplayer_pawn_athena "/MedCloud/Gameplay/Items/BP_SmokeCloud_PawnComp.BP_SmokeCloud_PawnComp_C")) self)
            (CallFunction|RefreshforMultiCloud (Actor|GetComponentbyClass _asplayer_pawn_athena "/MedCloud/Gameplay/Items/BP_SmokeCloud_PawnComp.BP_SmokeCloud_PawnComp_C"))))))))

(fn GetDynamicDestructionResourceType
  (return "Metal"))

(fn GetDynamicDestructionStaticMeshComponents
  (return (Utilities|Array|MakeArray (Variables|Default|GetStaticMesh)) 0 0))

(fn HandleResumeSimulationChanged
  (|CallParentFunction)
  (Networking|SwitchHasAuthority
    (:Remote
      (if (Variables|Default|GetbResumeSimulation)
        (bind _returnvalue (Utilities|Time|SetTimerbyFunctionName 0 "RefreshMaterialPosition" 0.1 true true))
        (Variables|Default|SetRefreshPositionTimerHandler _returnvalue)
        (else
          (Utilities|Time|ClearandInvalidateTimerbyHandle (Variables|Default|GetRefreshPositionTimerHandler))
          (CallFunction|RefreshMaterialPosition))))))

(fn InitializeSmokeParameters
  (bind _fx (Variables|Default|GetFX))
  (bind _sm_medmist_sphere (Variables|Default|GetSM_MedMist_Sphere))
  (Utilities|FlowControl|Sequence
    (:then_0
      (bind _returnvalue (Rendering|Material|CreateDynamicMaterialInstance "/MedCloud/Assets/MedMistSmokeVolume/M_MedMist_PostProcess_Camera_smoke.M_MedMist_PostProcess_Camera_smoke" "None" "None" :!class /Script/Engine.KismetMaterialLibrary))
      (bind _output_get (Variables|Default|SetPostProcessMaterial _returnvalue))
      (Variables|/Script/Engine.PostProcessComponent|SetSettings (Utilities|Struct|MakePostProcessSettings (Utilities|Struct|MakeWeightedBlendables (Utilities|Array|MakeArray (Utilities|Struct|MakeWeightedBlendable 1.0 _output_get)))) (Variables|Default|GetMedMist_PostProcess))
      (CallFunction|RefreshMaterialPosition))
    (:then_1
      (Rendering|Material|SetDefaultCustomPrimitiveDataVector3 _sm_medmist_sphere 3 (Math|Conversions|ToVector(LinearColor) (Variables|SmokeFX|GetSmokeColor)))
      (Rendering|Material|SetDefaultCustomPrimitiveDataFloat _sm_medmist_sphere 2 (Variables|SmokeFX|GetSmoke_Night_Multiplier))
      (Rendering|Material|SetDefaultCustomPrimitiveDataVector3 _sm_medmist_sphere 12 (Math|Conversions|ToVector(LinearColor) (Variables|SmokeFX|GetSmokeNightColorMultiplier)))
      (Rendering|Material|SetDefaultCustomPrimitiveDataVector3 _sm_medmist_sphere 8 (Math|Conversions|ToVector(LinearColor) (Variables|SmokeFX|GetSmokeEveningColorMultiplier)))
      (CallFunction|SmokeFade (and (Editor|InEditororPIE) (Variables|Default|GetSmokeDebug))))
    (:then_2
      (Niagara|SetNiagaraVariableByString(LinearColor) _fx "SmokeColor" (Variables|SmokeFX|GetSmokeColor))
      (Niagara|SetNiagaraVariableByString(Float) _fx "NightMultiplier" (Variables|SmokeFX|GetSmoke_Night_Multiplier)))))

(fn OnRep_SetFXMode
  (if (>= (Variables|Default|GetSetFXMode) 0)
    (if (>= (Variables|Default|GetSetFXMode) 1)
      (CallFunction|PlayFX)
      (else
        (CallFunction|EndFX)
        (CallFunction|PlayDeathFX)))))

(fn RefreshForMultiCloud
  (params (CloudPawnComp :type BP%20Smoke%20Cloud%20Pawn%20Comp%20Object%20Reference))
  (LocalVariables|SetOverlappingCloudNumber (Utilities|Array|Length (Variables|/MedCloud/Gameplay/Items/BP_SmokeCloud_PawnComp.BP_SmokeCloud_PawnComp_C|GetOverlappingMedCloud (Variables|Default|GetCloudPawnComp))))
  (if (> (LocalVariables|GetOverlappingCloudNumber) 0)
    (Rendering|SetVisibility (Variables|Default|GetSM_MedMist_Sphere (Utilities|Array|Get(acopy) (Variables|/MedCloud/Gameplay/Items/BP_SmokeCloud_PawnComp.BP_SmokeCloud_PawnComp_C|GetOverlappingMedCloud (Variables|Default|GetCloudPawnComp)))) true true)
    (if (> (LocalVariables|GetOverlappingCloudNumber) 1)
      (for _index (range 1 (LocalVariables|GetOverlappingCloudNumber))
        (Rendering|SetVisibility (Variables|Default|GetSM_MedMist_Sphere (Utilities|Array|Get(acopy) (Variables|/MedCloud/Gameplay/Items/BP_SmokeCloud_PawnComp.BP_SmokeCloud_PawnComp_C|GetOverlappingMedCloud (Variables|Default|GetCloudPawnComp)) _index)) false true)))))

(fn RefreshMaterialPosition
  (bind _returnvalue (Transformation|GetActorLocation))
  (Rendering|Material|SetVectorParameterValue (Variables|Default|GetPostProcessMaterial) "ActorLocation" (Math|Conversions|ToLinearColor(Vector) _returnvalue) :!class /Script/Engine.MaterialInstanceDynamic)
  (Rendering|Debug|DrawDebugSphere _returnvalue 512.0))

(fn ShouldUseDynamicDestructionMaterial
  (return false))

(fn ShouldUseDynamicDestructionNiagaraSystem
  (return false))

(fn SmokeFade
  (params (FadeIn :type Boolean))
  (bind _returnvalue (Math|Float|SelectFloat 1.5 3.0 FadeIn))
  (bind _returnvalue_2 (Utilities|Time|GetGameTimeinSeconds))
  (bind _returnvalue_3 (not FadeIn))
  (bind _returnvalue_4 (Math|Conversions|ToFloat(Boolean) _returnvalue_3))
  (bind _sm_medmist_sphere (Variables|Default|GetSM_MedMist_Sphere))
  (bind _postprocessmaterial (Variables|Default|GetPostProcessMaterial))
  (Utilities|FlowControl|Sequence
    (:then_0
      (Rendering|Material|SetDefaultCustomPrimitiveDataFloat _sm_medmist_sphere 16 _returnvalue)
      (Rendering|Material|SetDefaultCustomPrimitiveDataFloat _sm_medmist_sphere 1 _returnvalue_4)
      (Rendering|Material|SetDefaultCustomPrimitiveDataFloat _sm_medmist_sphere 0 _returnvalue_2))
    (:then_1
      (Rendering|Material|SetScalarParameterValue _postprocessmaterial "TransitionLength" _returnvalue :!class /Script/Engine.MaterialInstanceDynamic)
      (Rendering|Material|SetScalarParameterValue _postprocessmaterial "InvertTimestamp" _returnvalue_4 :!class /Script/Engine.MaterialInstanceDynamic)
      (Rendering|Material|SetScalarParameterValue _postprocessmaterial "InTimestamp" _returnvalue_2 :!class /Script/Engine.MaterialInstanceDynamic))))

(fn ConstructionScript
  (|CallParentFunction)
  (CallFunction|InitializeSmokeParameters))
