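; Pawn component for the MedCloud item that records its owning player pawn when that pawn is locally viewed.
; On BeginPlay it casts the component's owner to PlayerPawn_Athena and checks IsLocallyViewed.
; If the check passes, it stores the pawn in OwnerPawn and prints its display name as debug output.
; Tick has no logic, and no healing or cloud behavior is implemented here.
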(event EventBeginPlay
  (bind _asplayer_pawn_athena (Utilities|Casting|CastToPlayerPawn_Athena (Components|GetOwner)))
  (if (Pawn|IsLocallyViewed _asplayer_pawn_athena)
    (bind _output_get (Variables|Default|SetOwnerPawn _asplayer_pawn_athena))
    (Development|PrintString (Utilities|GetDisplayName _output_get))))

(event EventTick
  (params (DeltaSeconds :type Float%20%28single-precision%29)))
