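; B_MedCloud is a Blueprint for the MedCloud gameplay item, with almost no graph logic of its own.
; ShouldShowSoundIndicator returns false for any Fort player controller, so it never asks for a sound indicator.
; The construction script only calls its parent, and BeginPlay and Tick are empty.
; The graphs show no healing or cloud behavior.
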(event EventBeginPlay)

(event EventTick
  (params (DeltaSeconds :type Float%20%28single-precision%29)))

(fn ShouldShowSoundIndicator
  (params (PlayerController :type Fort%20Player%20Controller%20Object%20Reference))
  (return false))

(fn ConstructionScript
  (|CallParentFunction))
